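# Animation Blueprint for the M_LRG_EmeraldGlassTransform player body, correcting deformation on a pose copied from the parent mesh.
# The pose runs through deform, pistol and arm Control Rigs, additive neck and upper-arm bone offsets, then a neck rig.
# Event-graph alphas track pistol use, head rotation relative to spine_05, crouching and delayed hang-glider state.
# Despite its name, inFrontendNoEmote enables head and neck correction outside the frontend or during emotes.
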AnimBlueprint /BRCosmetics/Characters/Player/Male/Large/Bodies/M_LRG_EmeraldGlassTransform/Meshes/M_LRG_EmeraldGlassTransform_AnimBP
target_skeleton=&/M_LRG_EmeraldGlassTransform_Skeleton
graphs [
 AnimGraph {nodes [
   "Transform (Modify) Bone - Bone: attach_cape"(Alpha=1, Translation="0,0,0", Rotation="0,0,0", Scale="1,1,1") {translationMode=BMM_Additive, alphaBoolBlend {blendOption=Linear}, boneToModify {boneName=attach_cape}}
   n1 = Control Rig(Source=Copy Pose From Mesh() {bUseAttachedParent=true}) {
    alphaBoolBlend {blendOption=Linear}
    controlRigAssetReference {blueprintRigClass=&/M_LRG_DeformRig_EmeraldGlassTransform_ControlRig1.M_LRG_DeformRig_EmeraldGlassTransform_ControlRig1_C}
   }
   n2 = Control Rig(Source=@n1, Alpha=Get using Pistol() -> (using Pistol:double)) {
    alphaBoolBlend {blendOption=Linear}
    controlRigAssetReference {blueprintRigClass=&/M_LRG_EmeraldGlassTransform_Pistol_CtrlRig.M_LRG_EmeraldGlassTransform_Pistol_CtrlRig_C}
   }
   n3 = Control Rig(Source=@n2) {
    alphaBoolBlend {blendOption=Linear}
    controlRigAssetReference {blueprintRigClass=&/M_LRG_EmeraldGlassTransform__arms_CtrlRig.M_LRG_EmeraldGlassTransform__arms_CtrlRig_C}
   }
   n4 = "Transform (Modify) Bone - Bone: neck_01"(
    ComponentPose=Local To Component(LocalPose=@n3)
    Alpha=Get headRotBack() -> (headRotBack:double)
    Translation="0,-4,2"
    Rotation="0,0,0"
    Scale="1,1,1"
   ) {translationMode=BMM_Additive, alphaBoolBlend {blendOption=Linear}, boneToModify {boneName=neck_01}, translation {y=-4, z=2}}
   n5 = "Transform (Modify) Bone - Bone: neck_01"(ComponentPose=@n4, Alpha=Get HeadRotFwd() -> (HeadRotFwd:float), Translation="0,2,-2", Rotation="0,0,0", Scale="1,1,1") {translationMode=BMM_Additive, alphaBoolBlend {blendOption=Linear}, boneToModify {boneName=neck_01}, translation {y=2, z=-2}}
   n6 = "Transform (Modify) Bone - Bone: neck_01"(ComponentPose=@n5, Alpha=Get isCrouchingHeadBack() -> (isCrouchingHeadBack:double), Translation="0,0,2", Rotation="0,0,15", Scale="1,1,1") {
    rotationMode=BMM_Additive
    translationMode=BMM_Additive
    alphaBoolBlend {blendOption=Linear}
    alphaScaleBiasClamp {bInterpResult=true, interpSpeedDecreasing=20, interpSpeedIncreasing=20}
    boneToModify {boneName=neck_01}
    rotation {roll=15}
    translation {z=2}
   }
   n7 = Get IsHangGlider() -> (IsHangGlider:double)
   n8 = "Transform (Modify) Bone - Bone: deform_upperarm_up_l"(ComponentPose=@n6, Alpha=@n7, Translation="9,0,0", Rotation="0,0,0", Scale="1,1,1") {translationMode=BMM_Additive, alphaBoolBlend {blendOption=Linear}, boneToModify {boneName=deform_upperarm_up_l}, translation {x=9}}
   n9 = "Transform (Modify) Bone - Bone: deform_upperarm_up_r"(ComponentPose=@n8, Alpha=@n7, Translation="-9,0,0", Rotation="0,0,0", Scale="1,1,1") {translationMode=BMM_Additive, alphaBoolBlend {blendOption=Linear}, boneToModify {boneName=deform_upperarm_up_r}, translation {x=-9}}
   n10 = Control Rig(Source=Component To Local(ComponentPose=@n9), Alpha=Get inFrontendNoEmote() -> (inFrontendNoEmote:float)) {
    alphaBoolBlend {blendOption=Linear}
    controlRigAssetReference {blueprintRigClass=&/M_LRG_EmeraldGlassTransform_Neck_CtrlRig.M_LRG_EmeraldGlassTransform_Neck_CtrlRig_C}
   }
   Output Pose(Result=@n10) {layerGroup=None}
  ]}
 EventGraph {nodes [
   Event BlueprintUpdateAnimation()
   > n1 = Sequence!(execute) -> (then_0, then_1, then_2:Exec)
   > n2 = Cast To FortPlayerAnimInstance(Object=GetAnimInstance(self=GetOwnerSkeletalMeshComponent()))
   > Set ar_firingPosition(ar_firingPosition:double=0) -> (Output_Get:double)
   > n3 = Switch on EFortWeaponCoreAnimation!(execute, Selection:EFortWeaponCoreAnimation Enum=Get WeaponCoreAnim(self=@n2)) -> (
    Melee:Exec
    Pistol:Exec
    Shotgun:Exec
    PaperBlueprint:Exec
    Rifle:Exec
    MeleeOneHand:Exec
    MachinePistol:Exec
    RocketLauncher:Exec
    GrenadeLauncher:Exec
    GoingCommando:Exec
    AssaultRifle:Exec
    TacticalShotgun:Exec
    SniperRifle:Exec
    TrapPlacement:Exec
    ShoulderLauncher:Exec
    AbilityDecoTool:Exec
    Crossbow:Exec
    C4:Exec
    RemoteControl:Exec
    DualWield:Exec
    AR_BullPup:Exec
    AR_ForwardGrip:Exec
    MedPackPaddles:Exec
    SMG_P90:Exec
    AR_DrumGun:Exec
    Consumable_Small:Exec
    Consumable_Large:Exec
    Balloon:Exec
    MountedTurret:Exec
    CreativeTool:Exec
    ExplosiveBow:Exec
    AshtonIndigo:Exec
    AshtonChicago:Exec
    MeleeDualWield:Exec
    Unarmed:Exec
   )
   Set using Pistol(
    execute=[
     @n3.AR_BullPup
     @n3.AR_DrumGun
     @n3.AR_ForwardGrip
     @n3.AbilityDecoTool
     @n3.AshtonChicago
     @n3.AshtonIndigo
     @n3.AssaultRifle
     @n3.Balloon
     @n3.C4
     @n3.Consumable_Large
     @n3.Consumable_Small
     @n3.CreativeTool
     @n3.Crossbow
     @n3.DualWield
     @n3.ExplosiveBow
     @n3.GoingCommando
     @n3.GrenadeLauncher
     @n3.MachinePistol
     @n3.MedPackPaddles
     @n3.Melee
     @n3.MeleeDualWield
     @n3.MeleeOneHand
     @n3.MountedTurret
     @n3.PaperBlueprint
     @n3.RemoteControl
     @n3.Rifle
     @n3.RocketLauncher
     @n3.SMG_P90
     @n3.Shotgun
     @n3.ShoulderLauncher
     @n3.SniperRifle
     @n3.TacticalShotgun
     @n3.TrapPlacement
    ]
    using Pistol:double=0
   ) -> (Output_Get:double)
   Set using Pistol(execute=@n3.Pistol, using Pistol=1)
   n4 = GetOwnerSkeletalMeshComponent()
   n5 = Get bIsInFrontEnd()
   n6 = "To Float (Boolean)"(InBool=OR Boolean(A=NOT Boolean(A=@n5), B=Get bIsPlayingEmote()))
   Set headRotBack(
    execute=@n1.then_1
    headRotBack:double=float * float(
     A=GetAlphaFromReferenceTransformDelta(SkeletalMeshComponent=@n4, BoneName=head, BaseName=spine_05, InputRange=(X=-20,Y=-50), TwistAxis="1,0,0")
     B=@n6
    ) {is=title K2Node_PromotableOperator}
   ) -> (Output_Get:double)
   n7 = GetOwnerSkeletalMeshComponent()
   > Set HeadRotFwd(HeadRotFwd:float=float * float(
     A=GetAlphaFromReferenceTransformDelta(SkeletalMeshComponent=@n7, BoneName=head, BaseName=spine_05, InputRange=(X=-20,Y=20), TwistAxis="1,0,0")
     B=@n6
    )) -> (Output_Get:float)
   > Set inFrontendNoEmote(inFrontendNoEmote:float=@n6) -> (Output_Get:float)
   > n8 = Set Is Crouching(Is Crouching:double="To Float (Boolean)"(InBool=Get bIsCrouching())) -> (Output_Get:double)
   > Set isCrouchingHeadBack(isCrouchingHeadBack:double=float * float(A=@n8, B=Get headRotBack())) -> (Output_Get:double)
   Event OnParachuteOpened()
   > Set Is Sky Diving(Is Sky Diving:double=0) -> (Output_Get:double)
   > n9 = Switch on EGliderType(Selection=Get PlayerGliderType() -> (PlayerGliderType=HangGlider))
   Set IsHangGlider(execute=[@n9.Cape, @n9.NoGlider, @n9.Stairs, @n9.Surfing, @n9.SurfingSimple, @n9.Umbrella], IsHangGlider:double=0) -> (Output_Get:double)
   Delay(execute=@n9.HangGlider, Duration=1)
   > Set IsHangGlider(IsHangGlider=1)
   Event OnParachuteClosed()
   > Set IsHangGlider(IsHangGlider=0)
  ]}
]
variables [
 ar_firingPosition {type=number}
 headRotBack {type=number}
 HeadRotFwd {type=number}
 inFrontendNoEmote {type=number}
 Is Crouching {type=number}
 Is Sky Diving {type=number}
 isCrouchingHeadBack {type=number}
 IsHangGlider {type=number}
 using Pistol {type=number}
]
