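# Control Rig for the face accessory part of the male large EmeraldGlassTransform body, correcting the mullet hair bone as the head moves.
# A rotation constraint with local-space offset blends the mullet between neck_01 and head, weighted toward the neck.
# A debug-drawn spherical pose reader on head, relative to neck_02, then weights an additional rotation and translation on the mullet.
# A second pose reader chain offsetting the head is not connected to execution.
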ControlRigBlueprint /BRCosmetics/Characters/Player/Male/Large/Bodies/M_LRG_EmeraldGlassTransform/Meshes/Parts/M_LRG_EmeraldGlassTransform_FaceAcc_CtrlRig
graphs [
 RigVMFunctionLibrary {links [], nodes []}
 RigVMModel {nodes [
   RigUnit_BeginExecution = &RigUnit_BeginExecution()
   &RigUnit_RotationConstraintLocalSpaceOffset(
    AdvancedSettings=(RotationOrderForFilter=XZY)
    Child=(Type=Bone,Name="mullet")
    Filter=(bX=True,bY=True,bZ=True)
    Parents=((Item=(Type=Bone,Name="neck_01"),Weight=0.75),(Item=(Type=Bone,Name="head"),Weight=0.25))
    Weight=1
    ExecutePin=@RigUnit_BeginExecution.ExecutePin
   )
   > n1 = &RigUnit_SphericalPoseReader(
    ActiveRegionScaleFactors=(PositiveWidth=0.25,NegativeWidth=0.25,PositiveHeight=0.25,NegativeHeight=0.25)
    ActiveRegionSize=0.1
    Debug=(bDrawDebug=True,DebugScale=25,DebugSegments=20,DebugThickness=0.25)
    DriverAxis=(X=0,Y=-1,Z=0)
    DriverItem=(Type=Bone,Name="head")
    FalloffRegionScaleFactors=(PositiveWidth=0.3,NegativeWidth=0.5,PositiveHeight=2,NegativeHeight=2)
    FalloffSize=0.2
    OptionalParentItem=(Type=Bone,Name="neck_02")
    RotationOffset=(X=-90,Y=0,Z=-30)
   )
   > &RigUnit_ModifyTransforms(
    ItemToModify=((Item=(Type=Bone,Name="mullet"),Transform=(Rotation=(X=0,Y=0,Z=0,W=1),Translation=(X=-1,Y=-1,Z=0),Scale3D=(X=1,Y=1,Z=1))))
    ItemToModify.0.Transform.Rotation=&RigVMFunction_MathQuaternionFromEuler(Euler=(X=-20,Y=10,Z=0)).Result
    Weight=@n1.OutputParam
   )
   n2 = &RigUnit_SphericalPoseReader(
    ActiveRegionScaleFactors=(PositiveWidth=0.25,NegativeWidth=0.25,PositiveHeight=0.25,NegativeHeight=0.25)
    ActiveRegionSize=0.1
    Debug=(bDrawDebug=True,DebugScale=25,DebugSegments=20,DebugThickness=0.25)
    DriverAxis=(X=0,Y=1,Z=0)
    DriverItem=(Type=Bone,Name="head")
    ExecutePin=()
    FalloffRegionScaleFactors=(PositiveWidth=1,NegativeWidth=1,PositiveHeight=1,NegativeHeight=1)
    FalloffSize=0.2
    OptionalParentItem=(Type=Bone,Name="neck_01")
    RotationOffset=(X=45,Y=0,Z=0)
   )
   > &RigUnit_ModifyTransforms(
    ItemToModify=((Item=(Type=Bone,Name="head"),Transform=(Rotation=(X=0,Y=0,Z=0,W=1),Translation=(X=0,Y=0,Z=5),Scale3D=(X=1,Y=1,Z=1))))
    ItemToModify.0.Transform.Rotation=&RigVMFunction_MathQuaternionFromEuler(Euler=(X=0,Y=0,Z=0)).Result
    Weight=@n2.OutputParam
   )
  ]}
]
hierarchy [
 root {loc=<0 0 0>, rot=<0 0 0 1>}
 pelvis {parent=root, loc=<0 -0.7321 92.4218>, rot=<0.0127 0.707 -0.0127 -0.707>}
 spine_01 {parent=pelvis, loc=<5.6002 -0.5128 0>, rot=<0 0 0.034 -0.9994>}
 spine_02 {parent=spine_01, loc=<8.0753 0.0139 0>, rot=<0 0 0.0238 -0.9997>}
 spine_03 {parent=spine_02, loc=<7.0393 0 0>, rot=<0 0 0.0149 -0.9999>}
 spine_04 {parent=spine_03, loc=<8.1024 0 0>, rot=<0 0 -0.0132 -0.9999>}
 spine_05 {parent=spine_04, loc=<10.0146 0 0>, rot=<0 0 -0.0015 -1>}
 clavicle_l {parent=spine_05, loc=<2.5856 -0.2866 -4.3777>, rot=<0.0465 -0.7289 -0.009 -0.6829>}
 neck_01 {parent=spine_05, loc=<10.815 1.9196 0>, rot=<0 0 -0.1864 -0.9825>}
 neck_02 {parent=neck_01, loc=<4.7347 0.3045 0>, rot=<0 0 0.0363 -0.9993>}
 head {parent=neck_02, loc=<5.7632 0.5552 0>, rot=<0 0 0.115 -0.9934>}
 HairA_l_1 {parent=head, loc=<11.967 6.0122 -0.4456>, rot=<0.1069 -0.3276 0.0335 -0.9382>}
 HairA_l_2 {parent=HairA_l_1, loc=<7 0 0>, rot=<-0.0039 -0.0077 0.0951 -0.9954>}
 HairA_l_3 {parent=HairA_l_2, loc=<7 0 0>, rot=<0.0228 0.1645 0.1356 -0.9767>}
 HairA_l_4 {parent=HairA_l_3, loc=<7 0 0>, rot=<0 -0.0239 0 -0.9997>}
 HairA_r_1 {parent=head, loc=<11.9675 6.0123 0.4456>, rot=<-0.3276 -0.1069 0.9382 0.0335>}
 HairA_r_2 {parent=HairA_r_1, loc=<-7 0 0>, rot=<-0.0039 -0.0077 0.0951 -0.9954>}
 HairA_r_3 {parent=HairA_r_2, loc=<-6.9997 0.0001 0.0003>, rot=<0.0228 0.1645 0.1356 -0.9767>}
 HairA_r_4 {parent=HairA_r_3, loc=<-7.0004 -0.0002 -0.0002>, rot=<0 -0.0239 0 -0.9997>}
 mullet {parent=head, loc=<11.7596 -8.4523 0>, rot=<-0.1259 0.6958 0.6958 -0.1259>}
 clavicle_r {parent=spine_05, loc=<2.5855 -0.2866 4.3777>, rot=<0.7289 0.0465 -0.6829 0.009>}
]
