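# Physics asset for the face accessory mesh of the large male EmeraldGlassTransform player body, setting up physics for its left and right hair bone chains.
# Each side has four capsule bodies chained by constraints, with both root segments attached to the head.
# The head body is a tiny offset capsule serving as the chains' anchor.
# All constraints share narrow swing limits and a zero twist limit.
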PhysicsAsset /BRCosmetics/Characters/Player/Male/Large/Bodies/M_LRG_EmeraldGlassTransform/Meshes/Parts/M_LRG_EmeraldGlassTransform_FaceAcc_Physics
bodies {
 HairA_l_1 [{shapeType=Capsule, length=3, radius=3, shapeName=HairA_l_1_capsule, center=<2 0 0>, rotation=<rot 90 0 0>}]
 HairA_l_2 [{shapeType=Capsule, length=3, radius=3, shapeName=HairA_l_2_capsule, center=<2 0 0>, rotation=<rot 90 0 0>}]
 HairA_l_3 [{shapeType=Capsule, length=3, radius=3, shapeName=HairA_l_3_capsule, center=<2 0 0>, rotation=<rot 90 0 0>}]
 HairA_l_4 [{shapeType=Capsule, length=3, radius=3, shapeName=HairA_l_4_capsule, center=<2 0 0>, rotation=<rot 90 0 0>}]
 HairA_r_1 [{shapeType=Capsule, length=3, radius=3, shapeName=HairA_r_1_capsule, center=<-2 0 0>, rotation=<rot 90 0 0>}]
 HairA_r_2 [{shapeType=Capsule, length=3, radius=3, shapeName=HairA_r_2_capsule, center=<-2 0 0>, rotation=<rot 90 0 0>}]
 HairA_r_3 [{shapeType=Capsule, length=3, radius=3, shapeName=HairA_r_3_capsule, center=<-2 0 0>, rotation=<rot 90 0 0>}]
 HairA_r_4 [{shapeType=Capsule, length=3, radius=3, shapeName=HairA_r_4_capsule, center=<-2 0 0>, rotation=<rot 90 0 0>}]
 head [{shapeType=Capsule, length=0.01, radius=0.01, shapeName=head_capsule, center=<8.2313 -9.5607 0.1788>, rotation=<rot 0 139.3881 -85.5432>}]
}
constraints [
 {bone1Name=HairA_l_2, bone2Name=HairA_l_1, swing1LimitDegrees=15, swing2LimitDegrees=7, twistLimitDegrees=0}
 {bone1Name=HairA_l_3, bone2Name=HairA_l_2, swing1LimitDegrees=15, swing2LimitDegrees=7, twistLimitDegrees=0}
 {bone1Name=HairA_l_4, bone2Name=HairA_l_3, swing1LimitDegrees=15, swing2LimitDegrees=7, twistLimitDegrees=0}
 {bone1Name=HairA_r_4, bone2Name=HairA_r_3, swing1LimitDegrees=15, swing2LimitDegrees=7, twistLimitDegrees=0}
 {bone1Name=HairA_r_2, bone2Name=HairA_r_1, swing1LimitDegrees=15, swing2LimitDegrees=7, twistLimitDegrees=0}
 {bone1Name=HairA_r_3, bone2Name=HairA_r_2, swing1LimitDegrees=15, swing2LimitDegrees=7, twistLimitDegrees=0}
 {bone1Name=HairA_l_1, bone2Name=head, swing1LimitDegrees=15, swing2LimitDegrees=7, twistLimitDegrees=0}
 {bone1Name=HairA_r_1, bone2Name=head, swing1LimitDegrees=15, swing2LimitDegrees=7, twistLimitDegrees=0}
]
