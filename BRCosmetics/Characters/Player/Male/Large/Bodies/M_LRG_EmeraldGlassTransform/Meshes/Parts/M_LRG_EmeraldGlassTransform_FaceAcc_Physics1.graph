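# Physics asset for the face accessory mesh of the large male EmeraldGlassTransform cosmetic body, set up for mullet back-hair physics.
# Three hair chains (center, left, right), each three capsules long, hang from a tiny mullet anchor on the head with narrow swing and zero twist limits.
# Larger capsules cover head, neck, clavicles and upper spine, linked by angularly free constraints, forming the body volume around the hair.
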PhysicsAsset /BRCosmetics/Characters/Player/Male/Large/Bodies/M_LRG_EmeraldGlassTransform/Meshes/Parts/M_LRG_EmeraldGlassTransform_FaceAcc_Physics1
bodies {
 clavicle_l [{shapeType=Capsule, length=22.5, radius=16.5, shapeName=clavicle_l_capsule, rotation=<rot 90 0 0>}]
 clavicle_r [{shapeType=Capsule, length=21.5, radius=17.5, shapeName=clavicle_r_capsule, rotation=<rot 90 0 0>}]
 dyn_hairBk_c_01 [{shapeType=Capsule, length=2, radius=2, shapeName=dyn_hairBk_c_01_capsule, center=<2 0 0>, rotation=<rot 90 0 0>}]
 dyn_hairBk_c_02 [{shapeType=Capsule, length=2, radius=2, shapeName=dyn_hairBk_c_02_capsule, center=<2 0 0>, rotation=<rot 90 0 0>}]
 dyn_hairBk_c_MSTR [{shapeType=Capsule, length=2, radius=2, shapeName=dyn_hairBk_c_MSTR_capsule, center=<2 0 0>, rotation=<rot 90 0 0>}]
 dyn_hairBk_l_01 [{shapeType=Capsule, length=2, radius=2, shapeName=dyn_hairBk_l_01_capsule, center=<2 0 0>, rotation=<rot 90 0 0>}]
 dyn_hairBk_l_02 [{shapeType=Capsule, length=2, radius=2, shapeName=dyn_hairBk_l_02_capsule, center=<2 0 0>, rotation=<rot 90 0 0>}]
 dyn_hairBk_l_MSTR [{shapeType=Capsule, length=2, radius=2, shapeName=dyn_hairBk_l_MSTR_capsule, center=<2 0 0>, rotation=<rot 90 0 0>}]
 dyn_hairBk_r_01 [{shapeType=Capsule, length=1.5, radius=2.5, shapeName=dyn_hairBk_r_01_capsule, center=<2 0 0>, rotation=<rot 90 0 0>}]
 dyn_hairBk_r_02 [{
   shapeType=Capsule
   length=2.25
   radius=2.25
   shapeName=dyn_hairBk_r_02_capsule
   center=<3.3984 -0.509 0>
   rotation=<rot 90 176.4237 -163.5765>
  }]
 dyn_hairBk_r_MSTR [{shapeType=Capsule, length=2, radius=2, shapeName=dyn_hairBk_r_MSTR_capsule, center=<2 0 0>, rotation=<rot 90 0 0>}]
 head [{shapeType=Capsule, length=5.5, radius=10.25, shapeName=head_capsule, rotation=<rot 90 0 -9.9999>}]
 mullet [{shapeType=Capsule, length=0.01, radius=0.01, shapeName=mullet_capsule}]
 neck_01 [{shapeType=Capsule, length=0, radius=11.75, shapeName=neck_01_capsule, rotation=<rot 90 0 0>}]
 neck_02 [{shapeType=Capsule, length=0, radius=11.75, shapeName=neck_02_capsule, rotation=<rot 90 0 0>}]
 spine_05 [{shapeType=Capsule, length=12, radius=21, shapeName=spine_05_capsule, rotation=<rot 0 -0.0001 0>}]
}
constraints [
 {bone1Name=dyn_hairBk_c_MSTR, bone2Name=mullet, swing1LimitDegrees=20, swing2LimitDegrees=20, twistLimitDegrees=0}
 {bone1Name=dyn_hairBk_c_01, bone2Name=dyn_hairBk_c_MSTR, swing1LimitDegrees=20, swing2LimitDegrees=20, twistLimitDegrees=0}
 {bone1Name=dyn_hairBk_c_02, bone2Name=dyn_hairBk_c_01, swing1LimitDegrees=20, swing2LimitDegrees=20, twistLimitDegrees=0}
 {bone1Name=dyn_hairBk_l_MSTR, bone2Name=mullet, swing1LimitDegrees=20, swing2LimitDegrees=20, twistLimitDegrees=0}
 {bone1Name=dyn_hairBk_l_01, bone2Name=dyn_hairBk_l_MSTR, swing1LimitDegrees=20, swing2LimitDegrees=20, twistLimitDegrees=0}
 {bone1Name=dyn_hairBk_l_02, bone2Name=dyn_hairBk_l_01, swing1LimitDegrees=20, swing2LimitDegrees=20, twistLimitDegrees=0}
 {bone1Name=dyn_hairBk_r_MSTR, bone2Name=mullet, swing1LimitDegrees=20, swing2LimitDegrees=20, twistLimitDegrees=0}
 {bone1Name=dyn_hairBk_r_01, bone2Name=dyn_hairBk_r_MSTR, swing1LimitDegrees=20, swing2LimitDegrees=10, twistLimitDegrees=0}
 {bone1Name=dyn_hairBk_r_02, bone2Name=dyn_hairBk_r_01, swing1LimitDegrees=20, swing2LimitDegrees=10, twistLimitDegrees=0}
 {
  bone1Name=mullet
  bone2Name=head
  swing1LimitDegrees=45
  swing1Motion=Free
  swing2LimitDegrees=45
  swing2Motion=Free
  twistLimitDegrees=45
  twistMotion=Free
 }
 {
  bone1Name=head
  bone2Name=neck_02
  swing1LimitDegrees=45
  swing1Motion=Free
  swing2LimitDegrees=45
  swing2Motion=Free
  twistLimitDegrees=45
  twistMotion=Free
 }
 {
  bone1Name=neck_02
  bone2Name=neck_01
  swing1LimitDegrees=45
  swing1Motion=Free
  swing2LimitDegrees=45
  swing2Motion=Free
  twistLimitDegrees=45
  twistMotion=Free
 }
 {
  bone1Name=clavicle_l
  bone2Name=spine_05
  swing1LimitDegrees=45
  swing1Motion=Free
  swing2LimitDegrees=45
  swing2Motion=Free
  twistLimitDegrees=45
  twistMotion=Free
 }
 {
  bone1Name=neck_01
  bone2Name=spine_05
  swing1LimitDegrees=45
  swing1Motion=Free
  swing2LimitDegrees=45
  swing2Motion=Free
  twistLimitDegrees=45
  twistMotion=Free
 }
 {
  bone1Name=clavicle_r
  bone2Name=spine_05
  swing1LimitDegrees=45
  swing1Motion=Free
  swing2LimitDegrees=45
  swing2Motion=Free
  twistLimitDegrees=45
  twistMotion=Free
 }
]
