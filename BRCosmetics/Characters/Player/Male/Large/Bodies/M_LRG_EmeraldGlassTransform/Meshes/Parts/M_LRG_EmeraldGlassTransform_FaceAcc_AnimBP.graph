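# Anim Blueprint for the EmeraldGlassTransform large male face accessory, adding secondary motion and pose correction to its hair bones.
# Copies the attached parent's pose through a Control Rig, then two head-space RigidBody simulations with separate physics assets; the second takes state-driven gravity from a deprecated function.
# Additive rotations adjust the mullet while crouch-moving and the right back hair when the head also tilts back.
# HeadRotFwd is computed but unused here.
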AnimBlueprint /BRCosmetics/Characters/Player/Male/Large/Bodies/M_LRG_EmeraldGlassTransform/Meshes/Parts/M_LRG_EmeraldGlassTransform_FaceAcc_AnimBP
target_skeleton=&/M_LRG_EmeraldGlassTransform_FaceAcc_Skeleton
graphs [
 AnimGraph {nodes [
   n1 = Control Rig(Source=Copy Pose From Mesh() {bUseAttachedParent=true}) {
    alphaBoolBlend {blendOption=Linear}
    controlRigAssetReference {blueprintRigClass=&/M_LRG_EmeraldGlassTransform_FaceAcc_CtrlRig.M_LRG_EmeraldGlassTransform_FaceAcc_CtrlRig_C}
   }
   n2 = RigidBody!(ComponentPose=Local To Component(LocalPose=@n1), Alpha=0.5, OverrideWorldGravity:Vector="0,0,0", ExternalForce) -> (Pose) {
    bClampLinearTranslationLimitToRefPose=true
    bOverrideWorldGravity=true
    overridePhysicsAsset=&/M_LRG_EmeraldGlassTransform_FaceAcc_Physics
    simulationSpace=BaseBoneSpace
    alphaBoolBlend {blendOption=Linear}
    baseBoneRef {boneName=head}
    componentAppliedLinearAccClamp=<3000 3000 3000>
    componentLinearAccScale=<0.4 0.3 0.3>
    componentLinearVelScale=<0.4 0.3 0.3>
    overrideWorldGravity {z=-980}
    simSpaceSettings {
     maxAngularAcceleration=1000
     maxAngularVelocity=3
     maxLinearAcceleration=250
     maxLinearVelocity=500
     velocityScaleZ=0
     worldAlpha=0.25
     externalLinearDragV {y=0.5}
    }
   }
   n3 = RigidBody(ComponentPose=@n2, OverrideWorldGravity=Get GravityOverride()) {
    bClampLinearTranslationLimitToRefPose=true
    bOverrideWorldGravity=true
    overridePhysicsAsset=&/M_LRG_EmeraldGlassTransform_FaceAcc_Physics1
    simulationSpace=BaseBoneSpace
    alphaBoolBlend {blendOption=Linear}
    baseBoneRef {boneName=head}
    componentAppliedLinearAccClamp=<3000 3000 3000>
    componentLinearAccScale=<0.4 0.3 0.3>
    componentLinearVelScale=<0.4 0.3 0.3>
    overrideWorldGravity {z=-980}
    simSpaceSettings {
     maxAngularAcceleration=1000
     maxAngularVelocity=3
     maxLinearAcceleration=250
     maxLinearVelocity=500
     velocityScaleZ=0
     worldAlpha=0.25
     externalLinearDragV {y=0.5}
    }
   }
   n4 = "Transform (Modify) Bone - Bone: mullet"(ComponentPose=@n3, Alpha=Get isCrouchingMoving() -> (isCrouchingMoving:double), Translation="0,0,0", Rotation="0,0,-15", Scale="1,1,1") {
    rotationMode=BMM_Additive
    translationMode=BMM_Additive
    alphaBoolBlend {blendOption=Linear}
    alphaScaleBiasClamp {bInterpResult=true}
    boneToModify {boneName=mullet}
    rotation {roll=-15}
   }
   n5 = "Transform (Modify) Bone - Bone: dyn_hairBk_r_MSTR"(ComponentPose=@n4, Alpha=Get isCrouchingHeadBack() -> (isCrouchingHeadBack:double), Translation="0,0,0", Rotation="0,0,15", Scale="1,1,1") {
    rotationMode=BMM_Additive
    translationMode=BMM_Additive
    alphaBoolBlend {blendOption=Linear}
    alphaScaleBiasClamp {bInterpResult=true}
    boneToModify {boneName=dyn_hairBk_r_MSTR}
    rotation {roll=15}
   }
   Output Pose(Result=Component To Local(ComponentPose=@n5)) {layerGroup=None}
  ]}
 EventGraph {nodes [
   Event BlueprintUpdateAnimation()
   > n1 = Sequence!(execute) -> (then_0, then_1, then_2:Exec)
   > n2 = Gravity Override State Function DEPRECATED(CharacterPartAnimInstance=Self-Reference(), GameTimeInSeconds=GetGameTimeInSeconds(), GravityOverrideParams=Get Gravity Override Params())
   > Set GravityOverride(GravityOverride=@n2)
   n3 = Set Is Crouching(execute=@n1.then_1, Is Crouching:double="To Float (Boolean)"(InBool=Get bIsCrouching())) -> (Output_Get:double)
   > n4 = Set Is Moving 2D(Is Moving 2D:double="To Float (Boolean)"(InBool=Get bIsMoving2D())) -> (Output_Get:double)
   > Set isCrouchingMoving(isCrouchingMoving:double=float * float(A=@n3, B=@n4) {is=title K2Node_PromotableOperator}) -> (Output_Get:double)
   Set headRotBack(
    execute=@n1.then_2
    headRotBack:double=GetAlphaFromReferenceTransformDelta(SkeletalMeshComponent=GetOwnerSkeletalMeshComponent(), BoneName=head, BaseName=spine_05, InputRange=(X=-20,Y=-50), TwistAxis="1,0,0")
   ) -> (Output_Get:double)
   > Set HeadRotFwd(HeadRotFwd:float=GetAlphaFromReferenceTransformDelta(SkeletalMeshComponent=GetOwnerSkeletalMeshComponent(), BoneName=head, BaseName=spine_05, InputRange=(X=-20,Y=20), TwistAxis="1,0,0")) -> (Output_Get:float)
   > Set isCrouchingHeadBack(isCrouchingHeadBack:double=float * float!(
     A=Get Is Crouching() -> (Is Crouching:double)
     B=Get headRotBack() -> (headRotBack:double)
     C:double=Get Is Moving 2D() -> (Is Moving 2D:double)
    ) -> (ReturnValue)) -> (Output_Get:double)
  ]}
]
variables [
 Gravity Override Params {type=GravityOverrideParamsStruct, default {skydive_FloatingDown {gravityOverride=<0 -700 100>}}}
 GravityOverride {type=Vector}
 headRotBack {type=number}
 HeadRotFwd {type=number}
 Is Crouching {type=number}
 Is Moving 2D {type=number}
 isCrouchingHeadBack {type=number}
 isCrouchingMoving {type=number}
]
